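# Additive unlit material for Niagara sprite and ribbon glow cards, stored with the Nautical Pajamas Underwater character's effects.
# A procedural radial falloff, scaled by particle alpha, is added to a rotatable flare texture whose exponent pulses over time.
# Dynamic parameters drive both layers; the result tints by particle color or an optional inner/outer two-tone gradient.
# View-angle and depth fades form an opacity term.
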Material /Game/Characters/Player/Male/Medium/Bodies/M_MED_Nautical_Pajamas_Underwater/Effects/M_SimpleGlowCard
graph {
 properties {
  bUsedWithNiagaraRibbons=true
  bUsedWithNiagaraSprites=true
  blendMode=BLEND_Additive
  pixelDepthOffsetMode=PDOM_Legacy
  shadingModel=MSM_Unlit
  usedShadingModels=Unlit
 }
 nodes [
  DynamicParameter() {parameterIndex=1, defaultValue=<rgba 1 0.05 0 0>, paramNames [Flare_PulseSpeed, Flare_PulseRange, Flare_Power, DepthFade]}
  DynamicParameter() {defaultValue=<rgba 2 0.2 1 0.1>, paramNames [Glow_Power, Flare_Opacity, Flare_Size, Flare_Rotation]}
  n1 = ParticleColor()
  n2 = TextureCoordinate()
  n3 = OneMinus(None=Distance(A=Constant(), B=ConstantBiasScale(None=@n2) {bias=-0.5, scale=2}))
  n4 = Multiply(
   A=Sign(None=@n3)
   B=Power(
    Base=@n3
    Exp=DynamicParameter() {defaultValue=<rgba 2 0.2 1 0.1>, paramNames [Glow_Power, Flare_Opacity, Flare_Size, Flare_Rotation]}.Glow_Power
   ) {constExponent=7}
  )
  n5 = DynamicParameter() {defaultValue=<rgba 2 0.2 1 0.1>, paramNames [Glow_Power, Flare_Opacity, Flare_Size, Flare_Rotation]}
  n6 = Multiply(A=ConstantBiasScale(None=TextureCoordinate()) {bias=-0.5, scale=2}, B=@n5.Flare_Size)
  n7 = TextureSampleParameter2D(UVs=MaterialFunctionCall(UVs=ConstantBiasScale(None=@n6), "Rotation Angle (0-1)"=Frac(None=@n5.Flare_Rotation)) {materialFunction=&CustomRotator}.Rotated Values) {parameterName=FlareMask, samplerSource=SSM_Clamp_WorldGroupSettings, texture=/Game/Effects/Fort_Effects/Textures/Flares/T_Flare_Spikeball}
  n8 = DynamicParameter() {parameterIndex=1, defaultValue=<rgba 1 0.05 0 0>, paramNames [Flare_PulseSpeed, Flare_PulseRange, Flare_Power, DepthFade]}
  n9 = Sine(None=Multiply(A=Time(), B=@n8.Flare_PulseSpeed))
  n10 = Multiply(A=Multiply(A=Add(A=@n9)) {constB=0.5}, B=Saturate(None=@n8.Flare_PulseRange)) {constB=0}
  n11 = Max(A=@n8.Flare_Power)
  n12 = LinearInterpolate(
   B=Power(Base=@n7.RGB, Exp=Add(A=@n10, B=@n11) {constB=2})
   Alpha=Max(A=DynamicParameter() {defaultValue=<rgba 2 0.2 1 0.1>, paramNames [Glow_Power, Flare_Opacity, Flare_Size, Flare_Rotation]}.Flare_Opacity) {constB=0}
  )
  n13 = Saturate(None=Add(A=Multiply(A=@n1.A, B=@n4), B=@n12))
  n14 = VectorParameter() {parameterName=OuterColor, defaultValue=<rgba 1 1 1 1>}
  n15 = VectorParameter() {parameterName=InnerColor, defaultValue=<rgba 1 1 1 1>}
  n16 = OneMinus(None=ScalarParameter() {defaultValue=1, group=TwoToneGlow, parameterName=InnerGlow_Size, sliderMax=1, sliderMin=0})
  n17 = Saturate(None=Multiply(A=Subtract(A=@n13, B=@n16), B=ScalarParameter() {defaultValue=1, group=TwoToneGlow, parameterName=InnerGlow_Sharpness, sliderMin=1}))
  MaterialFunctionCall(Emissive=LinearInterpolate(
    A=Multiply(A=@n1.RGB, B=@n13)
    B=LinearInterpolate(A=@n14.RGB, B=@n15.RGB, Alpha=@n17)
    Alpha=ScalarParameter() {group=TwoToneGlow, parameterName="USE_TwoToneGlow ", sliderMax=1, sliderMin=0}
   )) {materialFunction=&MF_FX_EmissiveAdjustments}
  n18 = DynamicParameter() {parameterIndex=1, defaultValue=<rgba 1 0.05 0 0>, paramNames [Flare_PulseSpeed, Flare_PulseRange, Flare_Power, DepthFade]}
  Multiply(
   A=MaterialFunctionCall("FadeSensitivity (optional)"=ScalarParameter() {parameterName="FadeSensitivity (angle)"}) {materialFunction=&FadeBasedOnViewAngle}.Result
   B=ShadingPathSwitch(Default=DepthFade(Opacity=@n13, FadeDistance=@n18.DepthFade)) {desc=Migrated from FeatureLevelSwitch}
  )
  Saturate(None=@n12)
 ]
}
